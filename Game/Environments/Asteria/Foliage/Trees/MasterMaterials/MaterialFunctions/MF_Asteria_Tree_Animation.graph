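# Material function that adds world position offset animation to Asteria tree materials, with a DisableAllAnimation switch that bypasses it by default.
# A static switch picks soft or rigid tree transform offsets, which feed into MF_Asteria_Tree_SimpleAnimation.
# An optional fade, on by default, blends the offset to zero by pixel depth.
# Under Nanite, NaniteReplace swaps that depth fade for MF_WPODisableBlend.
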MaterialFunction /Game/Environments/Asteria/Foliage/Trees/MasterMaterials/MaterialFunctions/MF_Asteria_Tree_Animation
graph {nodes [
  n1 = FunctionInput() {inputType=FunctionInput_MaterialAttributes}
  n2 = GetMaterialAttributes(None=@n1) {attributeGetTypes [#1]}
  n3 = MaterialFunctionCall!(StormMask, OptionalWPO, TreeAddOnOverallBend) {materialFunction=&/MF_WPO_Tree_Soft_Transform}
  n4 = MaterialFunctionCall!(TangentNormalToTransform=@n2.Normal, TimeInput, OptionalWPO, TreeAddOnOverallBend) {materialFunction=&/MF_WPO_Rigid_Transform}
  n5 = MaterialFunctionCall!(In=SetMaterialAttributes!(
    MaterialAttributes=@n2.MaterialAttributes
    WorldPositionOffset=StaticSwitchParameter(True=@n3.Offset, False=@n4.Offset) {group=Animation, parameterName=Use2024AnimationScipt}
    Normal=@n2.MaterialAttributes
   ) {attributeSetTypes [#2, #1]}) {materialFunction=&/MF_Asteria_Tree_SimpleAnimation}
  n6 = GetMaterialAttributes(None=@n5.Result) {attributeGetTypes [#2]}
  n7 = PixelDepth()
  n8 = ScalarParameter() {defaultValue=7500, group=WPO Fade, parameterName=WPO_Fade_Offset}
  n9 = ScalarParameter() {defaultValue=1500, group=WPO Fade, parameterName=WPO_Fade_Distance}
  n10 = NaniteReplace!(
   Default=Divide(A=Subtract(A=@n7.R, B=@n8), B=@n9)
   Nanite=MaterialFunctionCall!(FadeBlendDistance=@n9, AutoHandleBadDistance) {materialFunction=/Game/Global/MaterialFunctions/TechArt/MF_WPODisableBlend}.VertexShaderOutput
  )
  n11 = SetMaterialAttributes!(
   MaterialAttributes=@n6.MaterialAttributes
   WorldPositionOffset=LinearInterpolate(A=@n6.WorldPositionOffset, B=Constant3Vector(), Alpha=Saturate(None=@n10))
  ) {attributeSetTypes [#2]}
  FunctionOutput(None=StaticSwitchParameter(True=@n1, False=StaticSwitchParameter(True=@n11, False=@n5.Result) {defaultValue=true, group=WPO Fade, parameterName=UseWPOFade}) {defaultValue=true, group=Animation, parameterName=DisableAllAnimation})
  Reroute(None=@n4.Custom Tangent Normal)
 ]}
