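# Material function for Asteria tree trunks that builds bark attributes from color, normal and SMR textures, with optional tiled detail and per-attribute array overrides, and combines them with incoming leaf attributes.
# In texture-array or UDIM mode it blends plain trunk and leaf attributes, with a UV-derived trunk/leaf mask declared.
# Otherwise, IsTrunk selects leaf input or trunk layered through gradient, graffiti, RVT, moss and lichen functions.
# A second output exposes SMR green.
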MaterialFunction /Game/Environments/Asteria/Foliage/Trees/MasterMaterials/MaterialFunctions/MF_Asteria_Trunk
graph {nodes [
  Constant()
  n1 = FunctionInput(Preview=GetMaterialAttributes().MaterialAttributes) {bUsePreviewValueAsDefault=true, desc=Leaf Materail, inputType=FunctionInput_MaterialAttributes}
  n2 = Constant() {r=1}
  n3 = StaticSwitchParameter(True=@n2, False=StaticSwitchParameter(True=@n2, False=NamedRerouteUsage()) {group=TextureArray, parameterName=UseUDIM_Textures}) {group=TextureArray, parameterName=UseTextureArrays}
  n4 = Constant3Vector()
  n5 = NamedRerouteUsage()
  n6 = SetMaterialAttributes!(
   MaterialAttributes
   BaseColor=NamedRerouteUsage()
   Normal=NamedRerouteUsage()
   Specular=NamedRerouteUsage()
   Roughness=NamedRerouteUsage()
   Opacity=@n3
   OpacityMask=@n3
   SubsurfaceColor=StaticSwitchParameter(True=@n4, False=StaticSwitchParameter(True=@n4, False=@n5) {group=TextureArray, parameterName=UseUDIM_Textures}) {group=TextureArray, parameterName=UseTextureArrays}
  ) {attributeSetTypes [
    #1
    #2
    #3
    #4
    #5
    #6
    #7
   ]}
  n7 = BlendMaterialAttributes(A=@n6, B=@n1, Alpha=NamedRerouteUsage())
  n8 = FunctionInput(Preview=StaticBool()) {bUsePreviewValueAsDefault=true, inputName=IsImpostorBake, inputType=FunctionInput_StaticBool, sortPriority=20}
  n9 = MaterialFunctionCall!(
   In=MaterialFunctionCall!(
    In=MaterialFunctionCall!(In=@n6) {materialFunction=&/MF_TreeTrunk_GradientBlends}.Result
    IsImpostorBake=@n8
   ) {materialFunction=&/MF_TreeTrunk_Graffiti}.Result
   IsImpostorBake=@n8
  ) {materialFunction=&/MF_Asteria_Tree_RVTBlend}
  n10 = StaticSwitchParameter(
   True=MaterialFunctionCall!(In=MaterialFunctionCall!(In=@n9.Result) {materialFunction=&/MF_JungleMoss_Trees}.Result) {materialFunction=&/MF_Lichen_Trees}.Result
   False=@n1
  ) {group=Trunk, parameterName=IsTrunk}
  FunctionOutput(None=StaticSwitchParameter(True=@n7, False=StaticSwitchParameter(True=@n7, False=@n10) {group=TextureArray, parameterName=UseUDIM_Textures}) {group=TextureArray, parameterName=UseTextureArrays})
  FunctionOutput(None=NamedRerouteUsage()) {desc=Not used for Metallic. General purpose output., outputName=TrunkSMR G_Channel, sortPriority=1}
  n11 = TextureSampleParameter2D() {group=Trunk Textures, parameterName=Normal_Trunk, samplerType=SAMPLERTYPE_Normal, texture=/Engine/EngineMaterials/BaseFlattenNormalMap}
  MaterialFunctionCall(Normal=@n11.RGB, Flatness=ScalarParameter() {group=Trunk, parameterName=Trunk_Normal_Flatten}) {materialFunction=&FlattenNormal}
  n12 = ComponentMask(None=TextureCoordinate()) {r=true}
  n13 = Saturate(None=Ceil(None=Subtract(A=@n12) {constB=5}))
  NamedRerouteDeclaration(None=@n13) {name=TrunkLeafMask, nodeColor=<rgba 0.2188 0 1 1>}
  n14 = TextureSampleParameter2D() {
   group=Trunk Textures
   parameterName=SMR_Trunk
   samplerType=SAMPLERTYPE_Masks
   texture=/Game/Global/Textures/Default/Blanks/Temp_specular_Mask
  }
  n15 = MaterialFunctionCall(Float3=@n14.RGB) {materialFunction=&BreakOutFloat3Components}
  NamedRerouteDeclaration(None=@n15.G) {name=SMR_green, desc=Not used for Metallic. General purpose output., nodeColor=<rgba 0 0.25 1 1>}
  NamedRerouteDeclaration(None=Constant3Vector() {constant=<rgba 8 0.95 0 0>}) {name=Visual Warning 2, nodeColor=<rgba 1 0.6678 0 1>}
  n16 = Multiply(
   A=ComponentMask(None=VectorParameter() {group=TrunkDetail, parameterName=Detail Tiling, sortPriority=50, defaultValue=<rgba 4 2 0 1>}.RGB) {g=true, r=true}
   B=TextureCoordinate()
  )
  n17 = TextureSampleParameter2D(UVs=@n16) {
   group=TrunkDetail
   parameterName=Trunk_DetailMap
   samplerSource=SSM_Wrap_WorldGroupSettings
   samplerType=SAMPLERTYPE_Masks
   texture=/Game/Environments/Apollo/Foliage/Textures_Materials/Apollo_Bark_Detail_D_01
   channelNames {a="", b="", g="S[ecular Overlay", r=BaseColor Overlay}
  }
  n18 = Subtract(A=@n17.RGB) {constB=0.5}
  NamedRerouteDeclaration(None=LinearInterpolate(B=ComponentMask(None=@n18) {g=true}, Alpha=ScalarParameter() {defaultValue=0.25, group=TrunkDetail, parameterName=Detail Strength Specular}) {constA=0.5}) {name=DetailTexture_Specular, nodeColor=<rgba 0.4297 0 1 1>}
  n19 = TextureSampleParameter2D(UVs=@n16) {
   group=TrunkDetail
   parameterName=Trunk_Detail_Normal
   samplerSource=SSM_Wrap_WorldGroupSettings
   samplerType=SAMPLERTYPE_Normal
   texture=/Game/Environments/Apollo/Foliage/Textures_Materials/Apollo_Bark_Detail_N_01
  }
  NamedRerouteDeclaration(None=MaterialFunctionCall(Normal=@n19.RGB, Flatness=ScalarParameter() {group=TrunkDetail, parameterName=DetailNormalFlatten}) {materialFunction=&FlattenNormal}.Result) {name=DetailTexture_Normal, nodeColor=<rgba 0.4297 0 1 1>}
  NamedRerouteDeclaration(None=LinearInterpolate(B=ComponentMask(None=@n18) {r=true}, Alpha=ScalarParameter() {defaultValue=0.5, group=TrunkDetail, parameterName=Detail Strength}) {constA=0.5}) {name=DetailTexture, nodeColor=<rgba 0.4297 0 1 1>}
  n20 = TextureSampleParameter2D() {group=Trunk Textures, parameterName=BaseColor_Trunk}
  n21 = MaterialFunctionCall!(Float In=ScalarParameter() {defaultValue=1, group=Trunk, parameterName=BrightnessTrunk}, Vector_In, Index=Constant() {r=28}) {materialFunction=&/MF_AsteriaTree_ArrayAttribute}
  n22 = ScalarParameter() {group=Trunk, parameterName=DesatTrunk}
  n23 = Constant() {r=29}
  n24 = MaterialFunctionCall!(
   Leaves Input
   Trunk_Input=Desaturation(
    None=Multiply(A=@n20.RGB, B=@n21.Float Out)
    Fraction=MaterialFunctionCall(Float In=@n22, Index=@n23) {materialFunction=&MF_AsteriaTree_ArrayAttribute}.Float Out
   )
   UseCustomPrimValue
  ) {materialFunction=&/MF_Asteria_Tree_VertexColorBlend}
  n25 = NamedRerouteUsage()
  NamedRerouteDeclaration(None=StaticSwitchParameter(
    True=MaterialFunctionCall(Base=@n24.Trunk Output, Blend=@n25) {materialFunction=&Blend_Overlay}.Result
    False=@n24.Trunk Output
   ) {group=TrunkDetail, parameterName=UseTrunk DetailTextures}) {name=BaseColor, nodeColor=<rgba 1 0.5156 0 1>}
  NamedRerouteDeclaration(None=@n15.R) {name=SpecularTextureInput, nodeColor=<rgba 1 0 0.3281 1>}
  n26 = LinearInterpolate(
   A=MaterialFunctionCall(Float In=ScalarParameter() {group=Trunk, parameterName=Specular2_Trunk}, Index=Constant() {r=27}) {materialFunction=&MF_AsteriaTree_ArrayAttribute}.Float Out
   B=MaterialFunctionCall(Float In=ScalarParameter() {defaultValue=0.5, group=Trunk, parameterName=Specular_Trunk}, Index=Constant() {r=26}) {materialFunction=&MF_AsteriaTree_ArrayAttribute}.Float Out
   Alpha=NamedRerouteUsage()
  )
  n27 = NamedRerouteUsage()
  NamedRerouteDeclaration(None=StaticSwitchParameter(
    True=MaterialFunctionCall(Base=@n26, Blend=@n27) {materialFunction=&Blend_Overlay}.Result
    False=@n26
   ) {group=TrunkDetail, parameterName=UseTrunk DetailTextures}) {name=Specular, nodeColor=<rgba 0 0.1797 1 1>}
  n28 = ScalarParameter() {defaultValue=1, group=Trunk, parameterName=Roughness_Trunk}
  n29 = Constant() {r=25}
  NamedRerouteDeclaration(None=Multiply(
    A=MaterialFunctionCall(Float In=@n28, Index=@n29) {materialFunction=&MF_AsteriaTree_ArrayAttribute}.Float Out
    B=NamedRerouteUsage()
   )) {name=Roughness, nodeColor=<rgba 0.125 1 0 1>}
  n30 = NamedRerouteUsage()
  NamedRerouteDeclaration(None=StaticSwitchParameter(
    True=MaterialFunctionCall(BaseNormal=@n11.RGB, AdditionalNormal=@n30) {materialFunction=&BlendAngleCorrectedNormals}.Result
    False=@n11.RGB
   ) {group=TrunkDetail, parameterName=UseTrunk DetailTextures}) {name=Normal, nodeColor=<rgba 1 0.5156 0 1>}
  NamedRerouteDeclaration(None=@n15.B) {name=RoughnessTextureInput, nodeColor=<rgba 0.6172 1 0 1>}
  n31 = ComponentMask(None=WorldPosition().XYZ) {b=true}
  NamedRerouteDeclaration(None=Sine(None=Divide(A=@n31) {constB=50})) {name=VisualWaring 1, nodeColor=<rgba 1 0.7456 0 1>}
  Reroute(None=@n13)
 ]}
