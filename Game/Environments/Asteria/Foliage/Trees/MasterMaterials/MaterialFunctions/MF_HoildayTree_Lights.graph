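# Material function that adds holiday-light glow to Asteria tree materials, with static switches for full shading override, emissive only, or passthrough.
# Emissive either steps through three colors using time and floored-UV offsets or uses static/primitive-data color, masked by texture and sphere gradient with two-sided weighting.
# Fresnel, optional primitive-data disable, Lumen boost and inverse eye adaptation shape the glow before adding it to existing emissive.
# A leaf-pivot directional animation branch is left unconnected.
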MaterialFunction /Game/Environments/Asteria/Foliage/Trees/MasterMaterials/MaterialFunctions/MF_HoildayTree_Lights
graph {nodes [
  n1 = FunctionInput() {inputType=FunctionInput_MaterialAttributes}
  n2 = GetMaterialAttributes(None=@n1) {attributeGetTypes [#1]}
  n3 = NamedRerouteUsage()
  n4 = TextureSampleParameter2D() {
   group=HolidayShading
   parameterName=SMR
   samplerSource=SSM_Wrap_WorldGroupSettings
   texture=/Game/Characters/Fortress_Character/Textures/T_All_White
   channelNames {a="", b=Roughness, g=Metallic, r=Specular}
  }
  n5 = TextureSampleParameter2D() {
   group=HolidayShading
   parameterName=MaskTexture
   samplerSource=SSM_Wrap_WorldGroupSettings
   texture=&T_WhitesRGB
   channelNames {a="", b="", g="", r=OpacityMask}
  }
  n6 = SetMaterialAttributes!(
   MaterialAttributes=@n2.MaterialAttributes
   EmissiveColor=@n3
   Metallic=Multiply(A=@n4.Metallic, B=ScalarParameter() {defaultValue=1, group=HolidayShading, parameterName=HolidayShading_Metallic})
   OpacityMask=@n5.OpacityMask
   Specular=@n4.Specular
   Roughness=Multiply(A=@n4.Roughness, B=ScalarParameter() {defaultValue=1, group=HolidayShading, parameterName=HolidayShading_Roughness})
  ) {attributeSetTypes [
    #1
    #2
    #3
    #4
    #5
   ]}
  n7 = SetMaterialAttributes!(MaterialAttributes=@n2.MaterialAttributes, EmissiveColor=@n3) {attributeSetTypes [#1]}
  FunctionOutput(None=StaticSwitchParameter(True=@n6, False=StaticSwitchParameter(True=@n7, False=@n1) {group=HolidayShading, parameterName=UseHolidayTreeEmissive}) {group=HolidayShading, parameterName="UseHolidayTreeShading&&Emissive"})
  n8 = VectorParameter() {group=HolidayShading, parameterName=EmissiveColor_C, defaultValue=<rgba 0 0 1.8 1>}
  n9 = Time()
  n10 = ScalarParameter() {defaultValue=0.1, group=HolidayShading, parameterName=Emissive_Anim_Speed2}
  n11 = ComponentMask(None=TextureCoordinate() {coordinateIndex=1}) {r=true}
  n12 = TextureCoordinate() {coordinateIndex=2}
  n13 = AppendVector(A=ComponentMask(None=@n12) {r=true}, B=ComponentMask(None=@n12) {g=true})
  n14 = Multiply(A=Frac(None=AppendVector(A=@n11, B=@n13))) {constB=12000}
  n15 = Constant3Vector() {constant=<rgba 6000 6000 3000 0>}
  n16 = DotProduct(
   A=NamedRerouteDeclaration(None=Subtract(A=@n14, B=@n15) {constB=6000}) {name=LeafPivot Position, nodeColor=<rgba 0.0781 1 0 1>}
   B=Normalize(VectorInput=VectorParameter() {parameterName=EmissiveDirection, defaultValue=<rgba 0.5 0.5 -1 1>}.RGB)
  )
  n17 = ScalarParameter() {defaultValue=8278.4004, group=HolidayShading, parameterName=Emissive_Anim_Frequency2}
  n18 = Sine(None=Add(A=Multiply(A=@n9, B=@n10), B=Divide(A=@n16, B=@n17)))
  n19 = Multiply(A=@n18) {constB=2}
  n20 = Saturate(None=@n19)
  n21 = Multiply(A=@n19) {constB=-1}
  LinearInterpolate(
   A=LinearInterpolate(
    A=VectorParameter() {group=HolidayShading, parameterName=EmissiveColor_A, defaultValue=<rgba 1 0.0898 0 1>}.RGB
    B=@n8.RGB
    Alpha=Round(None=@n20)
   )
   B=@n8.RGB
   Alpha=Round(None=Saturate(None=@n21))
  )
  n22 = DotProduct(A=Multiply(A=VertexNormalWS(), B=TwoSidedSign()), B=CameraVectorWS())
  n23 = ScalarParameter() {defaultValue=1.7632, group=HolidayShading, parameterName=Emissive_FresnelExp}
  n24 = ScalarParameter() {defaultValue=6.2743, group=HolidayShading, parameterName=Emissive_FresnelInnerExp}
  n25 = Multiply(A=Time(), B=ScalarParameter() {defaultValue=0.4, group=HolidayShading, parameterName=Emissive_Anim_Speed})
  n26 = Floor(None=ComponentMask(None=TextureCoordinate()) {r=true})
  n27 = Sine(None=Divide(A=Add(A=@n25, B=@n26), B=ScalarParameter() {defaultValue=4.5056, group=HolidayShading, parameterName=Emissive_Anim_Frequency}))
  n28 = LinearInterpolate(
   A=VectorParameter() {group=HolidayShading, parameterName=EmissiveColor_B, defaultValue=<rgba 0.0242 1 0 1>}.RGB
   B=VectorParameter() {group=HolidayShading, parameterName=EmissiveColor_A, defaultValue=<rgba 1 0.0898 0 1>}.RGB
   Alpha=Round(None=Saturate(None=@n27))
  )
  n29 = VectorParameter() {group=HolidayShading, parameterName=EmissiveColor_C, defaultValue=<rgba 0 0 1.8 1>}
  n30 = Round(None=Saturate(None=Multiply(A=@n27) {constB=-1}))
  n31 = ScalarParameter() {defaultValue=30, group=HolidayShading, parameterName=Emissive_Brightness}
  n32 = VectorParameter() {group=HolidayShading, parameterName=EmissiveColor, defaultValue=<rgba 1 0.338 0 1>}
  n33 = VectorParameter() {bUseCustomPrimitiveData=true, group=HolidayShading, parameterName=EmissiveColor_primData, defaultValue=<rgba 1 0.338 0 1>}
  n34 = ScalarParameter() {group=HolidayShading, parameterName=Emissive_primDataColor}
  n35 = StaticSwitchParameter(
   True=Multiply(A=LinearInterpolate(A=@n28, B=@n29.RGB, Alpha=@n30), B=@n31)
   False=Multiply(A=LinearInterpolate(A=@n32.RGB, B=@n33.RGB, Alpha=@n34), B=@n31)
  ) {defaultValue=true, group=HolidayShading, parameterName=UseHolidayTreeEmissiveAnimation}
  n36 = TextureSampleParameter2D() {
   group=HolidayShading
   parameterName=Emissive
   samplerSource=SSM_Wrap_WorldGroupSettings
   texture=&T_All_White
  }
  n37 = Add(
   A=Multiply(A=NamedRerouteUsage(), B=@n35)
   B=Multiply(A=@n35, B=ChannelMaskParameter(None=@n36.RGB) {group=HolidayShading, parameterName=Emissive_MaskChannel}.R)
  )
  n38 = OneMinus(None=ScalarParameter() {bUseCustomPrimitiveData=true, defaultValue=1, parameterName=DisableEmissive, primitiveDataIndex=31})
  n39 = Multiply(
   A=Add(A=Power(Base=@n22, Exp=@n23), B=Power(Base=@n22, Exp=@n24))
   B=StaticSwitchParameter(True=Multiply(A=@n37, B=@n38), False=@n37) {defaultValue=true, group=HolidayShading, parameterName=UsePrimDataEmssiveScripting}
  )
  n40 = RayTracingQualitySwitch(Normal=@n39, RayTraced=Multiply(A=@n39, B=ScalarParameter() {defaultValue=4, group=HolidayShading, parameterName=Emissive_LumenBoost}))
  NamedRerouteDeclaration(None=Add(A=@n2.EmissiveColor, B=EyeAdaptationInverse(LightValueInput=@n40).EyeAdaptationInverse)) {name=outEmissive, nodeColor=<rgba 0.7344 1 0 1>}
  n41 = ScalarParameter() {group=HolidayShading, parameterName=Emissive_Backface, sortPriority=35}
  n42 = ScalarParameter() {defaultValue=1, group=HolidayShading, parameterName=Emissive_Frontface, sortPriority=35}
  n43 = Saturate(None=TwoSidedSign())
  NamedRerouteDeclaration(None=Multiply(A=NamedRerouteUsage(), B=LinearInterpolate(A=@n41, B=@n42, Alpha=@n43))) {name=twoSidedEmissive, nodeColor=<rgba 0.9453 0 1 1>}
  n44 = MaterialFunctionCall!(
   Location=ObjectPositionWS()
   Offset=VectorParameter() {group=HolidayShading, parameterName=HolidaySphereMaskOffset}.RGB
   Radius=ScalarParameter() {defaultValue=250, group=HolidayShading, parameterName=HolidaySphereMaskRadius}
   Calculate Camera Inside
   Depth Biased Alpha
   FadeDistance
  ) {materialFunction=/Engine/Functions/Engine_MaterialFunctions01/Gradient/SphereGradient-3D}
  n45 = ScalarParameter() {defaultValue=1, group=HolidayShading, parameterName=HolidaySphereMaskContrast}
  NamedRerouteDeclaration(None=Multiply(A=Power(Base=@n44.Result 0-1, Exp=@n45), B=ScalarParameter() {group=HolidayShading, parameterName=HolidaySphereMaskAmount})) {name=sphereMask, nodeColor=<rgba 0 1 0.2969 1>}
  Saturate(None=Divide(A=Add(A=@n19)))
  SetMaterialAttributes(EmissiveColor=@n18) {attributeSetTypes [#1]}
  VectorParameter() {group=HolidayShading, parameterName=EmissiveColor_B, defaultValue=<rgba 0.0242 1 0 1>}
 ]}
