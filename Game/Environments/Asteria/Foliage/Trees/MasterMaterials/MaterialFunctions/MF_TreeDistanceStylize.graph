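# Tree material function that stylizes incoming material attributes by distance, also outputting a StyleTransition mask built from pixel depth.
# Lerps adjust base color and subsurface brightness, roughness and normals, with spherified normals derived from world position and optional two-sided sign correction.
# Opacity mask lerps toward solid, with a separate shadow-pass input via ShadowReplace.
# A variation texture, sampled by radial angle, relative height and position offset, forms contrast-adjusted color splotches.
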MaterialFunction /Game/Environments/Asteria/Foliage/Trees/MasterMaterials/MaterialFunctions/MF_TreeDistanceStylize
graph {nodes [
  n1 = GetMaterialAttributes(None=FunctionInput() {inputType=FunctionInput_MaterialAttributes}) {attributeGetTypes [
    #1
    #2
    #3
    #4
    #5
   ]}
  n2 = NamedRerouteUsage()
  n3 = Constant() {r=1}
  n4 = NamedRerouteUsage()
  n5 = NamedRerouteUsage()
  n6 = Multiply(A=@n5, B=ScalarParameter() {defaultValue=0.8, group=Stylization, parameterName=StylizeSubsurfaceBrightness})
  n7 = NamedRerouteUsage()
  n8 = NamedRerouteUsage()
  n9 = Multiply(A=@n8, B=ScalarParameter() {defaultValue=0.5, group=Stylization, parameterName=StylizeBaseColorBrightness})
  n10 = NamedRerouteUsage()
  n11 = LinearInterpolate(
   A=Constant3Vector() {constant=<rgba 0 0 1 0>}
   B=NamedRerouteUsage()
   Alpha=ScalarParameter() {defaultValue=1, group=Stylization, parameterName=SpherifyNormals}
  )
  n12 = StaticSwitchParameter(True=Multiply(A=@n11, B=TwoSidedSign()), False=@n11) {defaultValue=true, group=Stylization, parameterName=TwoSidedSign}
  n13 = SetMaterialAttributes!(
   MaterialAttributes=@n1.MaterialAttributes
   BaseColor=Multiply(A=LinearInterpolate(A=@n8, B=@n9, Alpha=@n10), B=NamedRerouteUsage())
   Normal=LinearInterpolate(A=NamedRerouteUsage(), B=NamedRerouteDeclaration(None=@n12) {name=VertexNormal, nodeColor=<rgba 0 1 0.625 1>}, Alpha=NamedRerouteUsage())
   OpacityMask=ShadowReplace(Default=LinearInterpolate(A=@n2, B=@n3, Alpha=@n4), Shadow=NamedRerouteUsage())
   Roughness=LinearInterpolate(A=NamedRerouteUsage(), B=ScalarParameter() {defaultValue=1, group=Stylization, parameterName=StyleRoughness}, Alpha=NamedRerouteUsage())
   SubsurfaceColor=Multiply(A=LinearInterpolate(A=@n5, B=@n6, Alpha=@n7), B=NamedRerouteUsage())
  ) {attributeSetTypes [
    #1
    #2
    #3
    #4
    #5
   ]}
  FunctionOutput(None=@n13)
  n14 = ScalarParameter() {defaultValue=1, group=Stylization, parameterName=Stylize}
  n15 = Subtract(A=PixelDepth().R, B=ScalarParameter() {defaultValue=3000, group=Stylization, parameterName=StyleDistance})
  n16 = Saturate(None=Divide(A=Max(B=@n15), B=ScalarParameter() {defaultValue=3000, group=Stylization, parameterName=StyleDistanceTransition}))
  FunctionOutput(None=NamedRerouteDeclaration(None=Multiply(A=@n14, B=@n16)) {name=StyleTransition, nodeColor=<rgba 0.7812 1 0 1>}) {outputName=StyleTransition}
  LinearInterpolate(B=Constant3Vector() {constant=<rgba 1 0 0 0>}, Alpha=NamedRerouteUsage())
  Multiply(A=TwoSidedSign(), B=Transform(None=VertexNormalWS()) {transformSourceType=TRANSFORMSOURCE_World, transformType=TRANSFORM_Tangent})
  NamedRerouteDeclaration(None=@n1.Normal) {name=incomingNormal, nodeColor=<rgba 1 0.5859 0 1>}
  NamedRerouteDeclaration(None=@n1.OpacityMask) {name=incomingOpacityMask, nodeColor=<rgba 0.6406 1 0 1>}
  NamedRerouteDeclaration(None=@n1.Roughness) {name=incomingRoughness, nodeColor=<rgba 0.0078 0 1 1>}
  NamedRerouteDeclaration(None=@n1.BaseColor) {name=incomingBaseColor, nodeColor=<rgba 0.4766 0 1 1>}
  NamedRerouteDeclaration(None=@n1.SubsurfaceColor) {name=incomingSubsurfaceColor, nodeColor=<rgba 0 0.7187 1 1>}
  n17 = Subtract(A=WorldPosition().XYZ, B=NamedRerouteUsage())
  NamedRerouteDeclaration(None=Transform(None=Normalize(VectorInput=@n17)) {transformSourceType=TRANSFORMSOURCE_World, transformType=TRANSFORM_Tangent}) {name=spherenormals, nodeColor=<rgba 1 0.5859 0 1>}
  n18 = NamedRerouteUsage()
  n19 = Multiply(
   A=ScalarParameter() {defaultValue=1, group=Stylization, parameterName=StylizeColorationSizeN}
   B=MaterialFunctionCall(Vector or UVs=ComponentMask(None=@n18) {g=true, r=true}) {materialFunction=&VectorToRadialValue}.Vector Converted to Angle
  )
  n20 = Subtract(
   A=MaterialFunctionCall(Float3=WorldPosition().XYZ) {materialFunction=&BreakOutFloat3Components}.B
   B=MaterialFunctionCall(Float3=NamedRerouteUsage()) {materialFunction=&BreakOutFloat3Components}.B
  )
  n21 = Multiply(A=ScalarParameter() {defaultValue=0.5, group=Stylization, parameterName=StylizeColorationSizeZ}, B=Divide(A=Abs(None=@n20)) {constB=2000})
  n22 = ComponentMask(None=NamedRerouteUsage()) {g=true, r=true}
  n23 = ScalarParameter() {defaultValue=2000, group=Stylization, parameterName=StylizeColorationActorPositionDivisor}
  n24 = TextureSample(UVs=Add(A=AppendVector(A=@n19, B=@n21), B=Divide(A=@n22, B=@n23) {constB=10000})) {samplerType=SAMPLERTYPE_LinearColor, texture=/Game/Environments/Asteria/Foliage/Trees/MasterMaterials/DefaultTextures/T_TreeVariation}
  n25 = Multiply(
   A=ScalarParameter() {defaultValue=4, group=Stylization, parameterName=StylizeColorContrast}
   B=ScalarParameter() {defaultValue=1, group=Stylization, parameterName=Stylize}
  )
  NamedRerouteDeclaration(None=Add(
    A=ScalarParameter() {defaultValue=0.5, group=Stylization, parameterName=StylizeColorOffset}
    B=MaterialFunctionCall(In=@n24.RGB, Power=@n25) {materialFunction=&SCurve}.Result
   ) {constA=0.5}) {name=ColorSplotches, nodeColor=<rgba 1 0.4688 0 1>}
  NamedRerouteDeclaration(None=ActorPositionWS()) {name=InstancePosition, nodeColor=<rgba 1 0.6094 0 1>}
  NamedRerouteUsage()
  TransformPosition(None=Constant3Vector()) {transformSourceType=TRANSFORMPOSSOURCE_Instance, transformType=TRANSFORMPOSSOURCE_World}
  VertexNormalWS()
 ]}
