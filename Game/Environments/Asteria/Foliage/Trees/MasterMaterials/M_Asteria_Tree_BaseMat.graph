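# Master two-sided subsurface material for Asteria trees, built for Nanite and instanced meshes.
# Leaf color, normals and shading combine textures, vertex-color per-leaf variation, world-position hue shifts and primitive/per-instance data, with most parameters routed through an array-attribute function.
# Material functions then layer trunk, animation, snow, holiday lights, hit wobble, Curie and creative-move effects.
# Static switches support impostor baking, while collection parameters substitute constants when shadows or reflections are off.
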Material /Game/Environments/Asteria/Foliage/Trees/MasterMaterials/M_Asteria_Tree_BaseMat
graph {
 properties {
  bUseMaterialAttributes=true
  bUsedWithInstancedStaticMeshes=true
  bUsedWithNanite=true
  opacityMaskClipValue=0.5
  pixelDepthOffsetMode=PDOM_Legacy
  previewMesh=/Engine/Transient.MaterialEditorMeshComponent_28
  shadingModel=MSM_Subsurface
  twoSided=true
  usedShadingModels=Subsurface
 }
 nodes [
  ActorPositionWS()
  Constant3Vector()
  n1 = TextureSample(UVs=NamedRerouteUsage()) {samplerSource=SSM_Wrap_WorldGroupSettings, samplerType=SAMPLERTYPE_Normal, texture=/Engine/EngineMaterials/RandomNormal2}
  n2 = ScalarParameter() {defaultValue=1, group=Shading, parameterName=Normal Random PerLeaf}
  n3 = MaterialFunctionCall!(Float In=@n2, Vector_In, Index=Constant() {r=2}) {materialFunction=/Game/Environments/Asteria/Foliage/Trees/MasterMaterials/MaterialFunctions/MF_AsteriaTree_ArrayAttribute}
  MaterialFunctionCall(Normal=@n1.RGB, Flatness=@n3.Float Out) {materialFunction=&FlattenNormal}
  n4 = SetMaterialAttributes!(
   MaterialAttributes
   Base Color=NamedRerouteUsage()
   Specular=NamedRerouteUsage()
   Roughness=NamedRerouteUsage()
   Emissive Color
   Opacity=NamedRerouteUsage()
   Normal=NamedRerouteUsage()
   Subsurface Color=NamedRerouteUsage()
   World Position Offset
   Ambient Occlusion=Constant() {r=1}
  ) {attributeSetTypes [
    #1
    #2
    #3
    #4
    #5
    #6
    #7
    #8
    #9
   ]}
  n5 = NamedRerouteUsage()
  n6 = MaterialFunctionCall!(
   In=MaterialFunctionCall!(In=MaterialFunctionCall!(In=@n4, IsImpostorBake=@n5) {materialFunction=/Game/Environments/Asteria/Foliage/Trees/MasterMaterials/MaterialFunctions/MF_Asteria_Trunk}.Result) {materialFunction=/Game/Environments/Asteria/Foliage/Trees/MasterMaterials/MaterialFunctions/MF_Asteria_Tree_Animation}.Result
   FlipTwoSidedSign=StaticBoolParameter() {group=Snow, parameterName=SnowFlipTwosidedNormal}
   IsTangentSpaceInput=StaticBool() {value=true}
   HasBackFlipFaceNormals=StaticBoolParameter() {parameterName=UseFlipBackFaceNormals}
  ) {materialFunction=/Game/Packages/Fortnite_UniqueMaterials_1/MaterialFunctions/Trees/MF_Tree_Snow}
  n7 = GetMaterialAttributes(None=MaterialFunctionCall!(In=MaterialFunctionCall!(ExtraSnowMask, Simple Mask, ForceUseInWSNormal, UseActorPosition, UseSnowMaskNormal, UseColorTest, UseColorTestRock, In=@n6.Result) {materialFunction=/Game/Packages/Fortnite_UniqueMaterials_1/MaterialFunctions/Trees/MF_Tree_Snow_Legacy}.Result) {materialFunction=/Game/Environments/Asteria/Foliage/Trees/MasterMaterials/MaterialFunctions/MF_HoildayTree_Lights}.Result) {attributeGetTypes [#8]}
  n8 = SetMaterialAttributes!(
   MaterialAttributes=@n7.MaterialAttributes
   World Position Offset=Add(
    A=@n7.World Position Offset
    B=MaterialFunctionCall() {materialFunction=&MF_FN_DestructibleDamage}.DamageAnimWPO
   )
  ) {attributeSetTypes [#8]}
  n9 = StaticSwitchParameter(True=@n8, False=StaticSwitchParameter(True=@n8, False=@n7.MaterialAttributes) {group=Trunk, parameterName=IsTrunk}) {group=Trunk, parameterName=ForceEnable_HitFXWobble}
  n10 = MaterialFunctionCall(MA In=MaterialFunctionCall!(
    In=ShadingPathSwitch(Default=@n9, Deferred=@n9, Forward=@n7.MaterialAttributes, Mobile=@n7.MaterialAttributes)
    UseSphericalMappingOverride=StaticBool() {value=true}
   ) {materialFunction=/Game/Packages/Fortress_SharedMaterials/Base_Material_Functions/MF_CurieFallbackEffect}.Result) {materialFunction=&MF_CreativeMoveTool}
  n11 = StaticSwitch(
   True=MaterialFunctionCall!(Material Attributes=@n10.MA Out, IsTangentSpaceNormal=StaticBool() {desc=Don't change, value=true}, In=Constant() {r=1}) {materialFunction=/Game/ImpostorBaker/MaterialFunctions/ImposterCaptureSwitch}.Result
   False=@n10.MA Out
   Value=NamedRerouteDeclaration(None=StaticBoolParameter() {parameterName=EnableForImpostorBake}) {name=IsImpostorBake, nodeColor=<rgba 1 0.3984 0 1>}
  )
  MaterialFunctionCall!(In=@n11) {materialFunction=/Game/Packages/Fortress_SharedMaterials/Cinematics/CINE_MF_EnvPostFX}
  Multiply()
  n12 = MaterialFunctionCall(Float3=TransformPosition(None=Constant3Vector()) {transformSourceType=TRANSFORMPOSSOURCE_Instance, transformType=TRANSFORMPOSSOURCE_World}) {materialFunction=&BreakOutFloat3Components}
  n13 = ScalarParameter() {defaultValue=13000, group=World Hue Variation, parameterName=Color Variation Distance WS, sortPriority=10}
  n14 = Constant() {r=13000}
  n15 = Sine(None=Divide(A=@n12.R, B=StaticSwitchParameter(True=@n13, False=@n14) {group=World Hue Variation, parameterName=Enable Global Hue Variation Tweaking}))
  n16 = ScalarParameter() {defaultValue=11000, group=World Hue Variation, parameterName=Color Variation Distance WS2, sortPriority=10}
  n17 = Constant() {r=11000}
  n18 = Sine(None=Divide(A=@n12.G, B=StaticSwitchParameter(True=@n16, False=@n17) {group=World Hue Variation, parameterName=Enable Global Hue Variation Tweaking}))
  NamedRerouteDeclaration(None=Divide(A=Add(A=@n15, B=@n18))) {name=PerActorRandom, nodeColor=<rgba 0 1 0.7891 1>}
  n19 = ScalarParameter() {defaultValue=0.5, group=Shading, parameterName=SubSurface SSS Opacity Variation}
  n20 = Constant() {r=4}
  n21 = Add(
   A=MaterialFunctionCall(Float In=ScalarParameter() {defaultValue=0.4, group=Shading, parameterName=SubSurface SSS Opacity}, Index=Constant() {r=3}) {materialFunction=&MF_AsteriaTree_ArrayAttribute}.Float Out
   B=Multiply(
    A=NamedRerouteUsage()
    B=MaterialFunctionCall(Float In=@n19, Index=@n20) {materialFunction=&MF_AsteriaTree_ArrayAttribute}.Float Out
   )
  )
  n22 = NamedRerouteUsage()
  NamedRerouteDeclaration(None=LinearInterpolate(A=Constant() {desc=Shadows Off SSS Opacity, r=0.875}, B=Multiply(A=@n21, B=@n22), Alpha=NamedRerouteUsage())) {name=SubSurfaceOpacity, nodeColor=<rgba 0 1 0.9063 1>}
  n23 = MaterialFunctionCall(Float In=ScalarParameter() {defaultValue=1, group=Shading, parameterName=Roughness}, Index=Constant() {r=7}) {materialFunction=&MF_AsteriaTree_ArrayAttribute}
  n24 = LinearInterpolate(
   A=MaterialFunctionCall(Float In=ScalarParameter() {defaultValue=1, group=Shading, parameterName=Roughness_LeafBottom}, Index=Constant() {r=8}) {materialFunction=&MF_AsteriaTree_ArrayAttribute}.Float Out
   B=@n23.Float Out
   Alpha=Saturate(None=TwoSidedSign())
  )
  n25 = LinearInterpolate(
   A=@n24
   B=MaterialFunctionCall(Float In=ScalarParameter() {defaultValue=1.5227, group=Shading, parameterName=Roughness_At_Distance}, Index=Constant() {r=32}) {materialFunction=&MF_AsteriaTree_ArrayAttribute}.Float Out
   Alpha=NamedRerouteUsage()
  )
  n26 = NamedRerouteUsage()
  NamedRerouteDeclaration(None=LinearInterpolate(A=Constant() {desc=RoughnessReflections Disabled, r=0.8}, B=Multiply(A=@n25, B=@n26), Alpha=NamedRerouteUsage())) {name=Roughness, nodeColor=<rgba 0 0.2031 1 1>}
  n27 = MaterialFunctionCall(Float In=ScalarParameter() {defaultValue=0.507, group=Shading, parameterName=Specular}, Index=Constant() {r=6}) {materialFunction=&MF_AsteriaTree_ArrayAttribute}
  n28 = LinearInterpolate(
   A=MaterialFunctionCall(Float In=ScalarParameter() {defaultValue=0.4622, group=Shading, parameterName=Specular_LeafBottom}, Index=Constant() {r=5}) {materialFunction=&MF_AsteriaTree_ArrayAttribute}.Float Out
   B=@n27.Float Out
   Alpha=Saturate(None=TwoSidedSign())
  )
  n29 = Multiply(A=NamedRerouteUsage(), B=@n28)
  n30 = ScalarParameter() {defaultValue=0.125, group=Shading, parameterName=Specular_PerLeaf_Random_Low}
  n31 = Constant() {r=36}
  n32 = ScalarParameter() {defaultValue=1.5, group=Shading, parameterName=Specular_PerLeaf_Random_High}
  n33 = Constant() {r=35}
  n34 = NamedRerouteUsage()
  n35 = Saturate(None=LinearInterpolate(
    A=MaterialFunctionCall(Float In=@n30, Index=@n31) {materialFunction=&MF_AsteriaTree_ArrayAttribute}.Float Out
    B=MaterialFunctionCall(Float In=@n32, Index=@n33) {materialFunction=&MF_AsteriaTree_ArrayAttribute}.Float Out
    Alpha=Saturate(None=@n34)
   ))
  NamedRerouteDeclaration(None=LinearInterpolate(A=Constant() {desc=Specular Reflections Disabled, r=0.1}, B=Multiply(A=@n29, B=@n35), Alpha=NamedRerouteUsage())) {name=Specular, nodeColor=<rgba 0.1484 1 0 1>}
  n36 = Multiply(A=Constant() {r=0.1}, B=NamedRerouteUsage())
  NamedRerouteDeclaration(None=StaticSwitch(True=Constant() {r=1}, False=Add(A=@n36), Value=NamedRerouteUsage()) {desc=For ImpostorBakeing}) {name=ValueVariation, nodeColor=<rgba 0 0.0625 1 1>}
  n37 = VectorParameter() {bUseCustomPrimitiveData=true, parameterName=Color1_Base_primdata}
  n38 = PerInstanceCustomData!(DefaultValue) {dataIndex=3}
  n39 = NamedRerouteDeclaration(None=Add(A=@n37.A, B=@n38)) {name=PrimDataBlendAlpha, nodeColor=<rgba 0 1 0.3203 1>}
  n40 = LinearInterpolate(
   A=MaterialFunctionCall(Vector_In=VectorParameter() {group=Color, parameterName=Color1_Base, defaultValue=<rgba 1 1 1 1>}.RGB, Index=Constant() {r=24}) {materialFunction=&MF_AsteriaTree_ArrayAttribute}.Vector_Out
   B=Add(A=PerInstanceCustomData3Vector!(DefaultValue), B=@n37.RGB)
   Alpha=@n39
  )
  n41 = VectorParameter() {group=Color VertexColor Tints, parameterName=Green Vertex Channel Color, sortPriority=4, defaultValue=<rgba 1 1 1 1>}
  n42 = VectorParameter() {bUseCustomPrimitiveData=true, desc=20 not used, parameterName=Color2_Base_primdata, primitiveDataIndex=17, defaultValue=<rgba 0 0 0 1>}
  n43 = LinearInterpolate(A=@n41.RGB, B=Add(A=PerInstanceCustomData3Vector() {dataIndex=17}, B=@n42.RGB), Alpha=@n39)
  n44 = ScalarParameter() {group=Color VertexColor Tints, parameterName=Green Vertex Channel Mask Value LOW, sortPriority=5}
  n45 = Constant() {r=11}
  n46 = ScalarParameter() {group=Color VertexColor Tints, parameterName=Green Vertex Channel Mask Value HIGH, sortPriority=5}
  n47 = Constant() {r=12}
  n48 = Saturate(None=LinearInterpolate(
    A=MaterialFunctionCall(Float In=@n44, Index=@n45) {materialFunction=&MF_AsteriaTree_ArrayAttribute}.Float Out
    B=MaterialFunctionCall(Float In=@n46, Index=@n47) {materialFunction=&MF_AsteriaTree_ArrayAttribute}.Float Out
    Alpha=VertexColor().G
   ))
  NamedRerouteDeclaration(None=StaticSwitch(True=Constant3Vector() {constant=<rgba 1 1 1 0>}, False=LinearInterpolate(A=@n40, B=@n43, Alpha=@n48), Value=NamedRerouteUsage()) {desc=For ImpostorBakeing}) {name=LeafColor, nodeColor=<rgba 0 1 0.5547 1>}
  n49 = ScalarParameter() {defaultValue=20000, group=DistanceBlend, parameterName=Distance}
  n50 = Constant() {r=33}
  n51 = ScalarParameter() {defaultValue=5000, group=DistanceBlend, parameterName=DistanceOffset}
  n52 = Constant() {r=34}
  NamedRerouteDeclaration(None=MaterialFunctionCall(
    Fade Length=MaterialFunctionCall(Float In=@n49, Index=@n50) {materialFunction=&MF_AsteriaTree_ArrayAttribute}.Float Out
    Fade Offset=MaterialFunctionCall(Float In=@n51, Index=@n52) {materialFunction=&MF_AsteriaTree_ArrayAttribute}.Float Out
   ) {materialFunction=&CameraDepthFade}.Result) {name=Distance Blend, nodeColor=<rgba 0 1 1 1>}
  n53 = TextureSampleParameter2D() {
   group=Textures
   parameterName=SSSR
   samplerSource=SSM_Wrap_WorldGroupSettings
   samplerType=SAMPLERTYPE_Masks
   texture=&T_WhiteMask
   channelNames {a="", b=Roughness, g=SubSurfaceScatter, r=Specular}
  }
  n54 = MaterialFunctionCall(Float3=@n53.RGB) {materialFunction=&BreakOutFloat3Components}
  NamedRerouteDeclaration(None=@n54.G) {name=SSSTextureInput, nodeColor=<rgba 0.8516 0 1 1>}
  NamedRerouteDeclaration(None=@n54.B) {name=RoughnessTextureInput, nodeColor=<rgba 0 1 0.5781 1>}
  NamedRerouteDeclaration(None=@n38) {name=IsPerIstanceData, nodeColor=<rgba 1 0.2109 0 1>}
  NamedRerouteDeclaration(None=VertexColor().R) {name=PerLeafRandom, nodeColor=<rgba 1 0 0.6308 1>}
  n55 = CollectionParameter() {collection=&FortniteMaterialParameters, parameterName=ShadowsEnabled}
  NamedRerouteDeclaration(None=@n55) {name=ShadowsEnabled, nodeColor=<rgba 0.3828 0 1 1>}
  n56 = CollectionParameter() {collection=&FortniteMaterialParameters, parameterName=ReflectionsEnabled}
  n57 = CollectionParameter() {
   collection=&FortniteMaterialParameters
   parameterName=GlobalIlluminationEnabled
  }
  NamedRerouteDeclaration(None=Multiply(A=@n56, B=@n57)) {name=ReflectionsEnabled, nodeColor=<rgba 1 0 0.6562 1>}
  NamedRerouteDeclaration(None=@n54.R) {name=SpecularMaskTexture, nodeColor=<rgba 1 0.8672 0 1>}
  n58 = TextureSampleParameter2D() {
   group=Textures
   parameterName=Normal
   samplerSource=SSM_Wrap_WorldGroupSettings
   samplerType=SAMPLERTYPE_Normal
   texture=/Engine/EngineMaterials/BaseFlattenNormalMap
  }
  n59 = MaterialFunctionCall(
   Normal=@n58.RGB
   Flatness=MaterialFunctionCall(Float In=ScalarParameter() {defaultValue=0.4, group=Shading, parameterName=NormalFlatten}, Index=Constant() {r=1}) {materialFunction=&MF_AsteriaTree_ArrayAttribute}.Float Out
  ) {materialFunction=&FlattenNormal}
  n60 = MaterialFunctionCall(BaseNormal=LinearInterpolate(A=Constant3Vector() {constant=<rgba 0 0 1 0>}, B=@n1.RGB, Alpha=@n2), AdditionalNormal=@n59.Result) {materialFunction=&BlendAngleCorrectedNormals}
  n61 = TwoSidedSign()
  n62 = Multiply(A=@n60.Result, B=@n61)
  n63 = LinearInterpolate(
   A=ScalarParameter() {defaultValue=-0.4, parameterName=FlipNormalOcclusionMask_Low}
   B=ScalarParameter() {defaultValue=1.5, parameterName=FlipNormalOcclusionMask_High}
   Alpha=VertexColor().B
  )
  n64 = Normalize(VectorInput=LinearInterpolate(A=@n60.Result, B=@n62, Alpha=Saturate(None=@n63)))
  n65 = AppendVector(A=Constant2Vector() {g=1, r=1}, B=@n61)
  n66 = ComponentMask(None=VertexNormalWS()) {b=true}
  n67 = Normalize(VectorInput=LinearInterpolate(A=@n60.Result, B=Multiply(A=@n60.Result, B=@n65), Alpha=Saturate(None=@n66)))
  NamedRerouteDeclaration(None=StaticSwitchParameter(
    True=StaticSwitchParameter(True=@n64, False=@n62) {parameterName=UseFlipBackFaceNormals_AOMask}
    False=StaticSwitchParameter(True=@n67, False=@n60.Result) {group=Shading, parameterName=FlipBackNormalsHackForUpFacing}
   ) {parameterName=UseFlipBackFaceNormals}) {name=Normal, nodeColor=<rgba 0.3727 0.4306 1 1>}
  n68 = ScalarParameter() {defaultValue=0.2312, group=OpacityMask, parameterName=OpacityThresholdAdd}
  n69 = NamedRerouteUsage()
  NamedRerouteDeclaration(None=Add(B=LinearInterpolate(B=@n68, Alpha=@n69))) {name=OpacityMaskFinal, nodeColor=<rgba 1 0.5859 0 1>}
  n70 = Constant()
  n71 = NamedRerouteUsage()
  n72 = ScalarParameter() {group=World Hue Variation, parameterName=World Hue Variation Bias}
  n73 = Multiply(
   A=StaticSwitchParameter(True=Add(A=@n71, B=@n72), False=@n71) {group=World Hue Variation, parameterName=Enable Global Hue Variation Tweaking}
   B=StaticSwitchParameter(
    True=ScalarParameter() {defaultValue=0.0125, group=World Hue Variation, parameterName=World Hue Variation Strength}
    False=Constant() {desc=Fixed value For Wold HueVariation, r=0.0125}
   ) {group=World Hue Variation, parameterName=Enable Global Hue Variation Tweaking}
  )
  n74 = NamedRerouteUsage()
  n75 = ScalarParameter() {defaultValue=-0.0645, group=Color VertexColor Tints, parameterName=Red Vertex Channel Hue Shift PerLeaf Random, sortPriority=1}
  n76 = ScalarParameter() {bUseCustomPrimitiveData=true, defaultValue=-0.03, group=Color, parameterName=Hue Variation_primdata, primitiveDataIndex=4}
  n77 = PerInstanceCustomData() {constDefaultValue=-0.03, dataIndex=4}
  n78 = NamedRerouteUsage()
  n79 = MaterialFunctionCall(Float In=LinearInterpolate(A=@n75, B=LinearInterpolate(A=@n76, B=@n77, Alpha=@n78), Alpha=NamedRerouteUsage()), Index=Constant()) {materialFunction=&MF_AsteriaTree_ArrayAttribute}
  n80 = NamedRerouteUsage()
  n81 = Multiply(A=Frac(None=Add(B=@n80) {constA=0.5}), B=@n80)
  NamedRerouteDeclaration(None=Add(A=StaticSwitch(True=@n70, False=@n73, Value=@n74) {desc=For ImpostorBakeing}, B=Multiply(A=@n79.Float Out, B=@n81))) {name=HueShift, nodeColor=<rgba 0 1 0.7891 1>}
  NamedRerouteDeclaration(None=MaterialFunctionCall!(In=@n4) {materialFunction=/Game/Environments/Asteria/Foliage/Trees/MasterMaterials/MaterialFunctions/MF_TreeDistanceStylize}.StyleTransition) {name=StyleTransition, nodeColor=<rgba 1 0.2812 0 1>}
  n82 = NamedRerouteUsage()
  n83 = TextureSampleParameter2D() {
   group=Textures
   parameterName=Diffuse
   samplerSource=SSM_Wrap_WorldGroupSettings
   texture=/Game/Characters/Fortress_Character/Textures/T_All_White
  }
  n84 = Multiply(A=NamedRerouteUsage(), B=@n83.RGB)
  n85 = MaterialFunctionCall!(
   Leaves Input=Multiply(
    A=NamedRerouteUsage()
    B=MaterialFunctionCall(Hue Shift Percentage=@n82, Texture=@n84) {materialFunction=&HueShift}.Result
   )
   Trunk_Input
   UseCustomPrimValue=NamedRerouteUsage()
  ) {materialFunction=/Game/Environments/Asteria/Foliage/Trees/MasterMaterials/MaterialFunctions/MF_Asteria_Tree_VertexColorBlend}
  n86 = Multiply(
   A=@n85.Leaves Output
   B=MaterialFunctionCall(Float In=ScalarParameter() {defaultValue=0.2, group=Shading, parameterName=SubSurface SSSColor}, Index=Constant() {r=21}) {materialFunction=&MF_AsteriaTree_ArrayAttribute}.Float Out
  )
  n87 = Constant() {desc=Shaodows Off SSS Multiplier, r=0.875}
  NamedRerouteDeclaration(None=LinearInterpolate(A=Multiply(A=@n86, B=@n87) {constB=0.5}, B=@n86, Alpha=NamedRerouteUsage())) {name=SubSurfaceColor, nodeColor=<rgba 1 0.9844 0 1>}
  n88 = ScalarParameter() {defaultValue=1, group=Shading, parameterName=Brightness}
  n89 = Constant() {r=20}
  NamedRerouteDeclaration(None=Multiply(
    A=MaterialFunctionCall(Float In=@n88, Index=@n89) {materialFunction=&MF_AsteriaTree_ArrayAttribute}.Float Out
    B=@n85.Leaves Output
   )) {name=BaseColor, nodeColor=<rgba 0 1 0.4844 1>}
  NamedRerouteUsage()
  n90 = Multiply(A=@n1.RGB, B=@n3.Float Out)
  n91 = Constant3Vector() {constant=<rgba 1 1 0 0>}
  Normalize(VectorInput=Add(A=@n59.Result, B=Multiply(A=@n90, B=@n91)))
  Reroute(None=@n38)
  ScalarParameter() {desc=debug, parameterName=Debug}
  StaticSwitchParameter(True=@n24, False=@n23.Float Out) {group=Shading, parameterName=TwoSidedRoughness}
  StaticSwitchParameter(True=@n28, False=@n27.Float Out) {group=Shading, parameterName=TwoSidedRoughness}
  n92 = VertexInterpolator(VS=MaterialFunctionCall!(
    Position
    IsTrees=StaticBool() {value=true}
    IsGrass
    IsMediumPlant
    IsMediumPlant_2024Anim
    IsMediumPlant2ndLevel
    MediumPlant_BasePivotPostion
   ) {materialFunction=/Game/Environments/Asteria/Foliage/Wind/MF_WindActorData}.StormMask)
  StaticSwitchParameter(
   True=SetMaterialAttributes!(MaterialAttributes=@n11, Base Color=@n92.PS, Subsurface Color=@n92.PS, Specular=Constant()) {attributeSetTypes [#1, #7, #2]}
   False=@n11
  ) {defaultValue=true, parameterName=Visualize StromRadius}
  n93 = ScalarParameter() {bUseCustomPrimitiveData=true, defaultValue=0.5, group=Color, parameterName=ValueVariation_primdata, primitiveDataIndex=5}
  n94 = MaterialFunctionCall(Float In=ScalarParameter() {defaultValue=0.5, group=Color Variation WorldSpace, parameterName=ValueVariation}, Index=Constant() {r=23}) {materialFunction=&MF_AsteriaTree_ArrayAttribute}
  n95 = NamedRerouteUsage()
  StaticSwitchParameter(
   True=@n93
   False=StaticSwitchParameter(True=LinearInterpolate(A=@n94.Float Out, B=@n93, Alpha=@n95), False=@n94.Float Out) {group=Color, parameterName=UsePrimitiveDataSeasonColoring}
  ) {group=TextureArray, parameterName=UseTextureArrays}
  n96 = MaterialFunctionCall(Float In=ScalarParameter() {defaultValue=-0.03, group=Color Variation WorldSpace, parameterName=Hue Variation}, Index=Constant() {r=22}) {materialFunction=&MF_AsteriaTree_ArrayAttribute}
  n97 = ScalarParameter() {bUseCustomPrimitiveData=true, defaultValue=-0.03, group=Color, parameterName=Hue Variation_primdata, primitiveDataIndex=4}
  n98 = NamedRerouteUsage()
  StaticSwitchParameter(
   True=StaticSwitchParameter(True=LinearInterpolate(A=@n96.Float Out, B=@n97, Alpha=@n98), False=@n96.Float Out) {group=Color, parameterName=UsePrimitiveDataSeasonColoring}
   False=@n97
  ) {group=TextureArray, parameterName=UseTextureArrays}
 ]
}
