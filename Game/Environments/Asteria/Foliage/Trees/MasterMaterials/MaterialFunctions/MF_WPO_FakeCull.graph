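# Material function that fakes distance culling for Asteria tree foliage through World Position Offset, with separate FarMesh and NearMesh attribute outputs.
# A camera-to-object SphereMask, sized by SwapDistance, collapses vertices to the object origin: far mesh inside the radius, near mesh outside.
# The collapse is added to incoming WPO, which ray tracing uses unmodified via RayTracingQualitySwitch.
# The "cull" lerp and ShadowReplace nodes feed no output.
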MaterialFunction /Game/Environments/Asteria/Foliage/Trees/MasterMaterials/MaterialFunctions/MF_WPO_FakeCull
graph {nodes [
  n1 = FunctionInput() {inputType=FunctionInput_MaterialAttributes}
  n2 = GetMaterialAttributes(None=@n1) {attributeGetTypes [#1]}
  n3 = Subtract(A=ObjectPositionWS(), B=WorldPosition().XYZ)
  n4 = SphereMask(A=CameraPositionWS(), B=ObjectPositionWS(), Radius=ScalarParameter() {defaultValue=5000, parameterName=SwapDistance})
  n5 = Add(A=@n2.WorldPositionOffset, B=Multiply(A=@n3, B=@n4))
  FunctionOutput(None=SetMaterialAttributes!(MaterialAttributes=@n1, WorldPositionOffset=RayTracingQualitySwitch(Normal=@n5, RayTraced=@n2.WorldPositionOffset)) {attributeSetTypes [#1]}) {outputName=FarMesh}
  n6 = Add(A=@n2.WorldPositionOffset, B=Multiply(A=@n3, B=OneMinus(None=@n4)))
  FunctionOutput(None=SetMaterialAttributes(MaterialAttributes=@n1, WorldPositionOffset=RayTracingQualitySwitch(Normal=@n6, RayTraced=@n2.WorldPositionOffset)) {attributeSetTypes [#1]}) {outputName=NearMesh}
  LinearInterpolate(A=Constant() {r=1}, B=@n4, Alpha=ScalarParameter() {parameterName=cull})
  ShadowReplace(Default=@n6, Shadow=@n2.WorldPositionOffset)
 ]}
