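# Nanite-enabled master material for Asteria tree NLOD meshes, defaulting to baked alder color, normal and packed MRS textures.
# Scaled color feeds base and subsurface color, while the MRS red channel alone drives metallic, specular and roughness.
# Vertex color red blends between Default Lit and Subsurface versions of the same attributes.
# The blended attributes pass through MF_WPO_FakeCull, which by its name fakes culling through World Position Offset.
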Material /Game/Environments/Asteria/Foliage/Trees/MasterMaterials/M_Asteria_Tree_NLOD
graph {
 properties {
  bUseMaterialAttributes=true
  bUsedWithNanite=true
  pixelDepthOffsetMode=PDOM_Legacy
  shadingModel=MSM_FromMaterialExpression
  usedShadingModels=Default Lit | Subsurface
 }
 nodes [
  n1 = TextureSampleParameter2D() {
   parameterName=MRS
   samplerType=SAMPLERTYPE_LinearColor
   texture=/Game/Environments/Asteria/Foliage/Trees/PrototypeTree/PrototypeTool/RTT/AlderTreeLargeCustomNormals_Solid_PackedMRS_CF106538
  }
  n2 = TextureSampleParameter2D() {
   parameterName=Normal
   samplerType=SAMPLERTYPE_Normal
   texture=/Game/Environments/Asteria/Foliage/Trees/PrototypeTree/PrototypeTool/RTT/AlderTreeLargeCustomNormals_Solid_NormalMap_C151DF31
  }
  n3 = TextureSampleParameter2D() {
   parameterName=Color
   texture=/Game/Environments/Asteria/Foliage/Trees/PrototypeTree/PrototypeTool/RTT/AlderTreeLargeCustomNormals_Solid_BaseColor_CA81EF64
  }
  n4 = SetMaterialAttributes!(
   MaterialAttributes
   Base Color=Multiply(A=@n3.RGB, B=ScalarParameter() {defaultValue=1.5, parameterName=BaseColor})
   Metallic=@n1.R
   Specular=@n1.R
   Roughness=@n1.R
   Opacity=ScalarParameter() {defaultValue=0.05, parameterName=Opacity}
   Normal=@n2.RGB
   Subsurface Color=Multiply(A=@n3.RGB, B=ScalarParameter() {defaultValue=1, parameterName=sss})
   Shading Model=ShadingModel()
  ) {attributeSetTypes [
    #1
    #2
    #3
    #4
    #5
    #6
    #7
    #8
   ]}
  n5 = ShadingModel() {shadingModel=MSM_Subsurface}
  MaterialFunctionCall!(In=BlendMaterialAttributes(
    A=@n4
    B=SetMaterialAttributes!(MaterialAttributes=@n4, Shading Model=@n5) {attributeSetTypes [#8]}
    Alpha=VertexColor().R
   )) {materialFunction=/Game/Environments/Asteria/Foliage/Trees/MasterMaterials/MaterialFunctions/MF_WPO_FakeCull}
 ]
}
